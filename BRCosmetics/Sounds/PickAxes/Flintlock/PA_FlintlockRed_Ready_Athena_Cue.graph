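# Sound cue for the ready sound of the red Flintlock pickaxe cosmetic in Athena.
# It picks one of three equally weighted swing waves, then a modulator raises pitch and slightly varies volume.
# The cue plays at reduced volume in the melee weapon sound class, using close projectile attenuation and the shared pickaxe-ready property sheet.
# It primes on load and overrides concurrency to prevent new instances once the limit is reached.
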SoundCue /BRCosmetics/Sounds/PickAxes/Flintlock/PA_FlintlockRed_Ready_Athena_Cue
root=Modulator_0
nodes [
 {id=Modulator_0, children [Random_5], properties {pitchMax=1.2, pitchMin=1.1, volumeMax=1.1, volumeMin=1}}
 {id=Random_5, children [WavePlayer_27, WavePlayer_29, WavePlayer_30], properties {weights [1,1,1]}}
 {id=WavePlayer_27, properties {soundWaveAssetPtr=&/PA_FlintlockRed_Swing_01}}
 {id=WavePlayer_29, properties {soundWaveAssetPtr=&/PA_FlintlockRed_Swing_02}}
 {id=WavePlayer_30, properties {soundWaveAssetPtr=&/PA_FlintlockRed_Swing_03}}
]
properties {
 attenuationSettings=&Fort_Weapon_Projectile_Close_Athena_Att
 audioPropertiesSheet=/BRCosmetics/Sounds/Settings/PropertySheets/Pickaxe/APS_Cosmetic_Pickaxe_Ready
 bOverrideConcurrency=true
 bPrimeOnLoad=true
 maxDistance=4950
 soundClassObject=&Fort_WeaponMelee
 volumeMultiplier=0.2
 concurrencyOverrides {resolutionRule=PreventNew}
}
